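; Burst gameplay cue for the Ice Gauntlets primary explosion, layering extra effects on the parent cue's burst.
; It spawns a radial force physics field at the hit when the physics field console variable is enabled.
; An extra Niagara explosion aligned to the hit normal spawns only off mobile and performance mode, above lowest detail.
; A gunshot sound indicator takes its range from the burst sound and targets hostile or neutral affiliations relative to the instigator's team.
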(fn OnBurst(EitherVFX-preferthisversion)
  (params (MyTarget :type Actor%20Object%20Reference) (Parameters :type Gameplay%20Cue%20Parameters%20Structure%20%28by%20ref%29) (ParticleComponents :type Array%20of%20FXSystem%20Component%20Object%20References) (AudioComponents :type Array%20of%20Audio%20Component%20Object%20References) (BurstCameraShakeInstance :type Camera%20Shake%20Base%20Object%20Reference) (BurstDecalInstance :type Decal%20Actor%20Object%20Reference))
  (bind (_normalizedmagnitude _rawmagnitude _effectcontext _matchedtagname _originaltag _aggregatedsourcetags _aggregatedtargettags _location _normal _instigator) (Ability|GameplayCue|BreakGameplayCueParameters Parameters))
  (|CallParentFunction MyTarget Parameters ParticleComponents AudioComponents BurstCameraShakeInstance BurstDecalInstance)
  (CallFunction|SpawnRadialForce MyTarget _location)
  (CallFunction|SpawnEpicFX MyTarget _normal _location)
  (CallFunction|SpawnSoundIndicator MyTarget _location _instigator)
  (return))

(fn SpawnEpicFX
  (params (WorldContextObject :type Actor%20Object%20Reference) (HitNormal :type Vector) (Location :type Vector))
  (if (Math|Boolean|ANDBoolean (not (Fortnite|IsMobilePlatform)) (not (Class|PlatformHelpersLibrary|IsPcPerfMode WorldContextObject)) (> (Rendering|GetRenderingDetailMode) 0))
    (Niagara|SpawnSystematLocation WorldContextObject (Variables|Default|GetEpicGraphicExplosionFX) Location (Math|Conversions|RotationFromXVector HitNormal) "1.000000,1.000000,1.000000" true true "AutoRelease")))

(fn SpawnRadialForce
  (params (WorldContextObject :type Object%20Reference) (InLocation :type Vector))
  (bind _returnvalue (Development|GetConsoleVariableIntValue "r.PhysicsField.EnableField"))
  (if (== _returnvalue 1)
    (Game|SpawnActorfromClass WorldContextObject "/Game/Effects/Niagara/Blueprints/PhsyicsFields/FN_RadialForce.FN_RadialForce_C" (Math|Conversions|ToTransform(Vector) InLocation) "Undefined" "MultiplyWithRoot" 0 (Variables|Default|GetRadialForce_Playrate) (Variables|Default|GetRadialForce_TimeDecay) (Variables|Default|GetRadialForce_ImpactRadius) (Variables|Default|GetRadialForce_Strength) 0.1 0)))

(fn SpawnSoundIndicator
  (params (WorldContextObject :type Actor%20Object%20Reference) (HitLocation :type Vector) (Instigator :type Actor%20Object%20Reference))
  (bind _bursteffectdata (Variables|BurstEffects|GetBurstEffectData))
  (bind (_burstparticles _burstsounds) (Utilities|Struct|BreakFortBurstEffectData _bursteffectdata))
  (bind _output (Utilities|Array|Get(aref) _burstsounds))
  (bind (_condition _attachment _soundcue) (Utilities|Struct|BreakFortGameplayCueAudioInfo _output))
  (if (Utilities|Array|IsValidIndex _burstsounds)
    (Utilities|IsValid _soundcue
      (:Is%20Valid
        (Fortnite|ShowSoundIndicatorLocalforTeam :WorldContextObject WorldContextObject :Location HitLocation :MaxAudibleDistance (Variables|/Script/Engine.SoundBase|GetMaxDistance _soundcue) :IndicatorType "Gunshot" :Team (Team|GetActorTeam Instigator) :Affiliations (Utilities|Array|MakeArray "Hostile" "Neutral") :Tint "(R=0.000000,G=0.000000,B=0.000000,A=0.000000)" :AttenuationMaxRangeMultiplier 1.0 :AttenuationIntensityMultiplier 1.0)))))
