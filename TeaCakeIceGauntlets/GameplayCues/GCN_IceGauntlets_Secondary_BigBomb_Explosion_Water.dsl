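; Gameplay cue notify for the water explosion of the Ice Gauntlets secondary big bomb, judging by its name.
; It declares a SpawnEpicFX function taking a world context actor, a hit normal and a location.
; The function has no body here, so the effect it spawns is not shown.
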(fn SpawnEpicFX
  (params (WorldContextObject :type Actor%20Object%20Reference) (HitNormal :type Vector :default "0, 0, 0") (Location :type Vector :default "0, 0, 0")))
